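# UMG header widget for the Scatter Objects editor tool, showing a single "Obj:" label beside a highlighted value field.
# A vertical box holds one horizontal row of two CommonTextBlocks styled with the editor default font.
# The blue value text, Sel1val, holds a placeholder and is exposed as a variable, as is the root box.
# The widget has no logic of its own; its PreConstruct, Construct and Tick events are empty.
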WidgetBlueprint /Game/Blueprints/Blutilities/ToolsWrkshp/ScrTools/ScatterObjects/SupportFiles/BP_UW_ScatterObjectsHeader
graphs [EventGraph {nodes [
   Event PreConstruct()
   Event Construct()
   Event Tick()
  ]}]
variables []
widgets {
 info {inheritedWidgetCount=0, parentClass=&UserWidget, rootWidgetClass=&VerticalBox, widgetCount=4}
 widgets [
  HeaderVertBox {widgetClassPath=&VerticalBox, bIsVariable=true, widget=HeaderVertBox}
  HorizontalBox_0 {widgetClassPath=&HorizontalBox, parent=HeaderVertBox, widget=HorizontalBox_0, slot {class=VerticalBoxSlot}}
  Sel1label {
   widgetClassPath=&CommonTextBlock
   parent=HorizontalBox_0
   widget=Sel1label
   properties {
    minDesiredWidth=90
    text=Obj:
    font {
     bForceMonospaced=false
     bMaterialIsStencil=false
     fontMaterial=null
     fontObject=/Engine/EditorResources/DefaultEditorFont
     letterSpacing=0
     monospacedWidth=1
     size=9.75
     skewAmount=0
     typefaceFontName=Regular
     outlineSettings {
      bApplyOutlineToDropShadows=false
      bMiteredCorners=false
      bSeparateFillAlpha=false
      outlineMaterial=null
      outlineSize=0
      outlineColor=<rgba 0 0 0 1>
     }
    }
    shadowOffset=<0 0>
   }
   slot {class=HorizontalBoxSlot}
  }
  Sel1val {
   widgetClassPath=&CommonTextBlock
   bIsVariable=true
   parent=HorizontalBox_0
   widget=Sel1val
   properties {
    text=XXX
    colorAndOpacity {colorUseRule=UseColor_Specified, specifiedColor=<rgba 0.0331 0.7011 1 1>}
    font {
     bForceMonospaced=false
     bMaterialIsStencil=false
     fontMaterial=null
     fontObject=&DefaultEditorFont
     letterSpacing=0
     monospacedWidth=1
     size=9.75
     skewAmount=0
     typefaceFontName=Regular
     outlineSettings {
      bApplyOutlineToDropShadows=false
      bMiteredCorners=false
      bSeparateFillAlpha=false
      outlineMaterial=null
      outlineSize=0
      outlineColor=<rgba 0 0 0 1>
     }
    }
    shadowOffset=<0 0>
   }
   slot {class=HorizontalBoxSlot}
  }
 ]
}
